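# Physics asset for the male medium Caveman cosmetic's shoes, set up for constrained motion of the fur on both feet.
# Each side has three fur capsules, each jointed to its own small base sphere with tight swing and tighter twist limits.
# A large master capsule per side and two small bracelet spheres have bodies but no constraints.
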PhysicsAsset /BRCosmetics/Characters/Player/Male/Medium/Bodies/M_MED_Caveman/Meshes/M_MED_Caveman_Shoes_Physics
bodies {
 dyn_bracelet_base_r [{shapeType=Sphere, radius=0.5}]
 dyn_bracelet_upper_base [{shapeType=Sphere, radius=0.5}]
 dyn_shoe_fur_01_base_l [{shapeType=Sphere, radius=0.5}]
 dyn_shoe_fur_01_base_r [{shapeType=Sphere, radius=0.5}]
 dyn_shoe_fur_01_parent_l [{shapeType=Capsule, length=4.345, radius=2.7348, center=<3.2777 0 0>, rotation=<rot 90 0 0>}]
 dyn_shoe_fur_01_parent_r [{shapeType=Capsule, length=4.2297, radius=3.0398, center=<3.1848 0.0001 0>, rotation=<rot 90 0 0>}]
 dyn_shoe_fur_02_base_l [{shapeType=Sphere, radius=0.5}]
 dyn_shoe_fur_02_base_r [{shapeType=Sphere, radius=0.5}]
 dyn_shoe_fur_02_parent_l [{shapeType=Capsule, length=2.6438, radius=2.5918, center=<3.0963 0 0>, rotation=<rot 90 0 0>}]
 dyn_shoe_fur_02_parent_r [{shapeType=Capsule, length=3.38, radius=2.13, center=<3.3478 0 0>, rotation=<rot 90 0 0>}]
 dyn_shoe_fur_03_base_l [{shapeType=Sphere, radius=0.5}]
 dyn_shoe_fur_03_base_r [{shapeType=Sphere, radius=0.5}]
 dyn_shoe_fur_03_parent_l [{shapeType=Capsule, length=3.1118, radius=2.3674, center=<3.3881 0 0>, rotation=<rot 90 0 0>}]
 dyn_shoe_fur_03_parent_r [{shapeType=Capsule, length=4.375, radius=2.255, center=<3.3566 -0.0001 0>, rotation=<rot 90 0 0>}]
 dyn_shoe_fur_master_l [{shapeType=Capsule, length=24.375, radius=5.625, rotation=<rot 90 -12.9074 -6.8942>}]
 dyn_shoe_fur_master_r [{shapeType=Capsule, length=24.375, radius=5.625, rotation=<rot 90 178.2101 174.733>}]
}
constraints [
 {bone1Name=dyn_shoe_fur_01_parent_l, bone2Name=dyn_shoe_fur_01_base_l, swing1LimitDegrees=10, swing2LimitDegrees=10, twistLimitDegrees=3}
 {bone1Name=dyn_shoe_fur_02_parent_l, bone2Name=dyn_shoe_fur_02_base_l, swing1LimitDegrees=10, swing2LimitDegrees=10, twistLimitDegrees=3}
 {bone1Name=dyn_shoe_fur_03_parent_l, bone2Name=dyn_shoe_fur_03_base_l, swing1LimitDegrees=10, swing2LimitDegrees=10, twistLimitDegrees=3}
 {bone1Name=dyn_shoe_fur_03_parent_r, bone2Name=dyn_shoe_fur_03_base_r, swing1LimitDegrees=10, swing2LimitDegrees=10, twistLimitDegrees=3}
 {bone1Name=dyn_shoe_fur_02_parent_r, bone2Name=dyn_shoe_fur_02_base_r, swing1LimitDegrees=10, swing2LimitDegrees=10, twistLimitDegrees=3}
 {bone1Name=dyn_shoe_fur_01_parent_r, bone2Name=dyn_shoe_fur_01_base_r, swing1LimitDegrees=10, swing2LimitDegrees=10, twistLimitDegrees=3}
]
