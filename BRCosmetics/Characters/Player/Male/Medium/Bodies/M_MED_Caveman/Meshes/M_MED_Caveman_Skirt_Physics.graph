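# Physics asset for the medium male Caveman body's skirt, defining secondary-motion bodies for the skirt strands and two skull pieces.
# Capsule chains hang from small sphere base bodies at the back, front and sides, with separate front and back skull chains.
# Constraints give each link narrow twist and varied swing limits, with the back chain loosening toward its tip.
# Unconstrained capsules on both thighs provide leg collision shapes.
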PhysicsAsset /BRCosmetics/Characters/Player/Male/Medium/Bodies/M_MED_Caveman/Meshes/M_MED_Caveman_Skirt_Physics
bodies {
 dyn_skirt_back_01 [{shapeType=Capsule, length=2.6414, radius=3.2013, center=<3.3851 0 0>, rotation=<rot 90 0 0>}]
 dyn_skirt_back_02 [{shapeType=Capsule, length=2.25, radius=3.0022, center=<2.4847 0 0>, rotation=<rot 90 0 0>}]
 dyn_skirt_back_03 [{shapeType=Capsule, length=3.9893, radius=3.4021, center=<4.0269 0 0>, rotation=<rot 90 0 0>}]
 dyn_skirt_back_base [{shapeType=Sphere, radius=0.5}]
 dyn_skirt_back_parent [{shapeType=Capsule, length=3.7547, radius=3.0445, center=<4.017 0 0>, rotation=<rot 90 0 0>}]
 dyn_skirt_front_l_01 [{shapeType=Capsule, length=5.7781, radius=2.029, center=<4.5308 0 -0.5276>, rotation=<rot 90 0 0>}]
 dyn_skirt_front_l_base [{shapeType=Sphere, radius=0.5}]
 dyn_skirt_front_l_parent [{shapeType=Capsule, length=3.9209, radius=2.3031, center=<3.3971 0 -0.61>, rotation=<rot 90 0 0>}]
 dyn_skirt_front_mid_01 [{shapeType=Capsule, length=4.6984, radius=2.1595, center=<4.0844 0 0>, rotation=<rot 90 0 0>}]
 dyn_skirt_front_mid_base [{shapeType=Sphere, radius=0.5}]
 dyn_skirt_front_mid_parent [{shapeType=Capsule, length=3.8389, radius=2.2034, center=<3.4126 0 0>, rotation=<rot 90 0 0>}]
 dyn_skirt_front_r_01 [{shapeType=Capsule, length=3.8766, radius=1.9427, center=<3.5583 0.182 0>, rotation=<rot 90 0 0>}]
 dyn_skirt_front_r_02 [{shapeType=Capsule, length=5.1683, radius=1.9985, center=<4.0746 0 0>, rotation=<rot 90 0 0>}]
 dyn_skirt_front_r_base [{shapeType=Sphere, radius=0.5}]
 dyn_skirt_front_r_parent [{shapeType=Capsule, length=5.3461, radius=1.9846, center=<4.4205 0 -0.0001>, rotation=<rot 83.6574 180 180>}]
 dyn_skirt_side_l_01 [{shapeType=Capsule, length=5.36, radius=1.7712, center=<4.4846 0 0>, rotation=<rot 90 0 0>}]
 dyn_skirt_side_l_base [{shapeType=Sphere, radius=0.5}]
 dyn_skirt_side_l_parent [{shapeType=Capsule, length=4.0312, radius=1.7104, center=<3.6127 0 0>, rotation=<rot 90 0 0>}]
 dyn_skirt_side_r_01 [{shapeType=Capsule, length=4.3846, radius=1.2654, center=<3.5691 0 0>, rotation=<rot 90 0 0>}]
 dyn_skirt_side_r_base [{shapeType=Sphere, radius=0.5}]
 dyn_skirt_side_r_parent [{shapeType=Capsule, length=4.68, radius=1.2981, center=<2.5007 0 0>, rotation=<rot 90 0 0>}]
 dyn_skull_back_base [{shapeType=Sphere, radius=0.5}]
 dyn_skull_back_parent [{shapeType=Capsule, length=3.0395, radius=5.6681, center=<6.2633 0 -2.382>, rotation=<rot 90 0 0>}]
 dyn_skull_front_base [{shapeType=Sphere, radius=0.5}]
 dyn_skull_front_parent [{shapeType=Capsule, length=4.7641, radius=5.9127, center=<5.0235 0 -3.1754>, rotation=<rot 90 0 0>}]
 thigh_l [{shapeType=Capsule, length=20.8292, radius=9.1708, center=<-16.1543 -1.005 -0.6224>, rotation=<rot 84.4822 0 0>}]
 thigh_r [{shapeType=Capsule, length=21.25, radius=8.75, center=<16.0317 0.9352 -0.0001>, rotation=<rot 84.9883 0 0>}]
}
constraints [
 {bone1Name=dyn_skull_back_parent, bone2Name=dyn_skull_back_base, swing1LimitDegrees=20, swing2LimitDegrees=20, twistLimitDegrees=3}
 {bone1Name=dyn_skull_front_parent, bone2Name=dyn_skull_front_base, swing1LimitDegrees=20, swing2LimitDegrees=20, twistLimitDegrees=3}
 {bone1Name=dyn_skirt_front_mid_01, bone2Name=dyn_skirt_front_mid_parent, swing1LimitDegrees=30, swing2LimitDegrees=30, twistLimitDegrees=3}
 {bone1Name=dyn_skirt_front_mid_parent, bone2Name=dyn_skirt_front_mid_base, swing1LimitDegrees=30, swing2LimitDegrees=30, twistLimitDegrees=3}
 {bone1Name=dyn_skirt_back_03, bone2Name=dyn_skirt_back_02, swing1LimitDegrees=20, swing2LimitDegrees=30, twistLimitDegrees=3}
 {bone1Name=dyn_skirt_back_02, bone2Name=dyn_skirt_back_01, swing1LimitDegrees=15, swing2LimitDegrees=30, twistLimitDegrees=3}
 {bone1Name=dyn_skirt_back_01, bone2Name=dyn_skirt_back_parent, swing1LimitDegrees=8, swing2LimitDegrees=10, twistLimitDegrees=3}
 {bone1Name=dyn_skirt_back_parent, bone2Name=dyn_skirt_back_base, swing1LimitDegrees=5, swing2LimitDegrees=10, twistLimitDegrees=3}
 {bone1Name=dyn_skirt_side_r_01, bone2Name=dyn_skirt_side_r_parent, swing1LimitDegrees=20, swing2LimitDegrees=25, twistLimitDegrees=3}
 {bone1Name=dyn_skirt_side_r_parent, bone2Name=dyn_skirt_side_r_base, swing1LimitDegrees=15, swing2LimitDegrees=15, twistLimitDegrees=3}
 {bone1Name=dyn_skirt_front_r_02, bone2Name=dyn_skirt_front_r_01, swing1LimitDegrees=40, swing2LimitDegrees=20, twistLimitDegrees=3}
 {bone1Name=dyn_skirt_front_r_01, bone2Name=dyn_skirt_front_r_parent, swing1LimitDegrees=40, swing2LimitDegrees=20, twistLimitDegrees=3}
 {bone1Name=dyn_skirt_front_r_parent, bone2Name=dyn_skirt_front_r_base, swing1LimitDegrees=20, swing2LimitDegrees=10, twistLimitDegrees=3}
 {bone1Name=dyn_skirt_side_l_01, bone2Name=dyn_skirt_side_l_parent, swing1LimitDegrees=30, swing2LimitDegrees=25, twistLimitDegrees=3}
 {bone1Name=dyn_skirt_side_l_parent, bone2Name=dyn_skirt_side_l_base, swing1LimitDegrees=50, swing2LimitDegrees=15, twistLimitDegrees=3}
 {bone1Name=dyn_skirt_front_l_01, bone2Name=dyn_skirt_front_l_parent, swing1LimitDegrees=20, swing2LimitDegrees=30, twistLimitDegrees=3}
 {bone1Name=dyn_skirt_front_l_parent, bone2Name=dyn_skirt_front_l_base, swing1LimitDegrees=20, swing2LimitDegrees=30, twistLimitDegrees=3}
]
