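# Post-process-style cosmetic AnimBP for the Caveman body, copying the attached parent mesh's pose and adding secondary deformation and motion.
# Arm and shoulder twist alphas, computed each update, drive additive corrective bone scaling; both biceps read the left-arm alpha, leaving the right one unused.
# Weighted three-bone constraints place the skirt-area skull bases.
# RigidBody nodes then simulate shoe and skirt physics assets, followed by AnimDynamics on two bracelet bones.
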AnimBlueprint /BRCosmetics/Characters/Player/Male/Medium/Bodies/M_MED_Caveman/Meshes/M_MED_Caveman_AnimBP
target_skeleton=&/M_MED_Caveman_Skeleton
graphs [
 AnimGraph {nodes [
   n1 = Get ArmRotate_L_Alpha() -> (ArmRotate_L_Alpha:double)
   n2 = "Transform (Modify) Bone - Bone: dyn_bicep_fr_l"(
    ComponentPose=Local To Component(LocalPose=Copy Pose From Mesh() {bUseAttachedParent=true})
    Alpha=@n1
    Translation="0,0,0"
    Rotation="0,0,0"
    Scale="1.15,1.15,1.15"
   ) {
    rotationMode=BMM_Additive
    rotationSpace=BCS_BoneSpace
    scaleMode=BMM_Additive
    scaleSpace=BCS_BoneSpace
    translationMode=BMM_Additive
    translationSpace=BCS_BoneSpace
    alphaBoolBlend {blendOption=Linear}
    alphaScaleBiasClamp {bInterpResult=true, interpSpeedDecreasing=4, interpSpeedIncreasing=6, inRange {max=0, min=-100}, outRange {max=0, min=1}}
    boneToModify {boneName=dyn_bicep_fr_l}
    scale=<1.15 1.15 1.15>
   }
   n3 = "Transform (Modify) Bone - Bone: dyn_bicep_bk_l"(ComponentPose=@n2, Alpha=@n1, Translation="0,0,0", Rotation="0,0,0", Scale="4,2,2") {
    rotationMode=BMM_Additive
    rotationSpace=BCS_BoneSpace
    scaleMode=BMM_Additive
    scaleSpace=BCS_BoneSpace
    translationMode=BMM_Additive
    translationSpace=BCS_BoneSpace
    alphaBoolBlend {blendOption=Linear}
    alphaScaleBiasClamp {interpSpeedDecreasing=4, interpSpeedIncreasing=6, inRange {max=0, min=-100}, outRange {max=0, min=1}}
    boneToModify {boneName=dyn_bicep_bk_l}
    scale=<4 2 2>
   }
   n4 = Get ArmRotate_L_Alpha()
   n5 = "Transform (Modify) Bone - Bone: dyn_bicep_bk_r"(ComponentPose=@n3, Alpha=@n4, Translation="0,0,0", Rotation="0,0,0", Scale="4,2,2") {
    rotationMode=BMM_Additive
    rotationSpace=BCS_BoneSpace
    scaleMode=BMM_Additive
    scaleSpace=BCS_BoneSpace
    translationMode=BMM_Additive
    translationSpace=BCS_BoneSpace
    alphaBoolBlend {blendOption=Linear}
    alphaScaleBiasClamp {interpSpeedDecreasing=4, interpSpeedIncreasing=6, inRange {max=0, min=-100}, outRange {max=0, min=1}}
    boneToModify {boneName=dyn_bicep_bk_r}
    scale=<4 2 2>
   }
   n6 = "Transform (Modify) Bone - Bone: dyn_bicep_fr_r"(ComponentPose=@n5, Alpha=@n4, Translation="0,0,0", Rotation="0,0,0", Scale="1.15,1.15,1.15") {
    rotationMode=BMM_Additive
    rotationSpace=BCS_BoneSpace
    scaleMode=BMM_Additive
    scaleSpace=BCS_BoneSpace
    translationMode=BMM_Additive
    translationSpace=BCS_BoneSpace
    alphaBoolBlend {blendOption=Linear}
    alphaScaleBiasClamp {bInterpResult=true, interpSpeedDecreasing=4, interpSpeedIncreasing=6, inRange {max=0, min=-100}, outRange {max=0, min=1}}
    boneToModify {boneName=dyn_bicep_fr_r}
    scale=<1.15 1.15 1.15>
   }
   n7 = "Transform (Modify) Bone - Bone: dyn_shoulder_r"(ComponentPose=@n6, Alpha=Get Shoulder_R_Alpha() -> (Shoulder_R_Alpha:double), Translation="0,0,0", Rotation="0,0,0", Scale="8,8,8") {
    rotationMode=BMM_Additive
    rotationSpace=BCS_BoneSpace
    scaleMode=BMM_Additive
    scaleSpace=BCS_BoneSpace
    translationMode=BMM_Additive
    translationSpace=BCS_BoneSpace
    alphaBoolBlend {blendOption=Linear}
    alphaScaleBiasClamp {interpSpeedDecreasing=4, interpSpeedIncreasing=6, inRange {max=0, min=-100}, outRange {max=0, min=1}}
    boneToModify {boneName=dyn_shoulder_r}
    scale=<8 8 8>
   }
   n8 = "Constraint - dyn_skull_front_base (PPP)"!(ComponentPose=@n7, Alpha=1, ConstraintWeights_0=0.7507, ConstraintWeights_1:float=0.1694, ConstraintWeights_2:float=0.0798) -> (Pose) {
    alphaBoolBlend {blendOption=Linear}
    boneToModify {boneName=dyn_skull_front_base}
    constraintSetup [{targetBone {boneName=pelvis}}, {targetBone {boneName=thigh_twist_01_r}}, {targetBone {boneName=spine_01}}]
    constraintWeights [1,1,1]
   }
   n9 = "Constraint - dyn_skull_back_base (PPP)"!(ComponentPose=@n8, Alpha=1, ConstraintWeights_0=0.8628, ConstraintWeights_1:float=0.0653, ConstraintWeights_2:float=0.0719) -> (Pose) {
    alphaBoolBlend {blendOption=Linear}
    boneToModify {boneName=dyn_skull_back_base}
    constraintSetup [{targetBone {boneName=pelvis}}, {targetBone {boneName=thigh_twist_01_r}}, {targetBone {boneName=spine_01}}]
    constraintWeights [1,1,1]
   }
   n10 = RigidBody(ComponentPose=@n9) {
    bClampLinearTranslationLimitToRefPose=true
    bOverrideWorldGravity=true
    overridePhysicsAsset=&/M_MED_Caveman_Shoes_Physics
    simulationSpace=BaseBoneSpace
    alphaBoolBlend {blendOption=Linear}
    baseBoneRef {boneName=root}
    componentAppliedLinearAccClamp=<6000 6000 6000>
    componentLinearAccScale=<0.25 0.25 0.25>
    componentLinearVelScale=<0.25 0.25 0.25>
    overrideWorldGravity {z=-980}
   }
   n11 = RigidBody(ComponentPose=@n10) {
    bClampLinearTranslationLimitToRefPose=true
    bOverrideWorldGravity=true
    overridePhysicsAsset=&/M_MED_Caveman_Skirt_Physics
    simulationSpace=BaseBoneSpace
    alphaBoolBlend {blendOption=Linear}
    alphaScaleBiasClamp {interpSpeedDecreasing=2, interpSpeedIncreasing=3}
    baseBoneRef {boneName=root}
    componentAppliedLinearAccClamp=<3000 3000 3000>
    componentLinearAccScale=<0.25 0.25 0.15>
    componentLinearVelScale=<0.25 0.25 0.15>
    overrideWorldGravity {z=-980}
   }
   n12 = Anim Dynamics - Bone: dyn_bracelet_parent_r(ComponentPose=@n11, Alpha=1, ExternalForce="0,0,0") {
    angularDampingOverride=0.9
    bGravityOverrideInSimSpace=true
    bOverrideAngularDamping=true
    bOverrideLinearDamping=true
    linearDampingOverride=0.99
    alphaBoolBlend {blendOption=Linear}
    boundBone {boneName=dyn_bracelet_parent_r}
    componentAppliedLinearAccClamp=<6000 6000 6000>
    componentLinearAccScale=<0.15 0.15 0.15>
    componentLinearVelScale=<0.15 0.15 0.15>
    physicsBodyDefinitions [{
      sphereCollisionRadius=0
      boundBone {boneName=dyn_bracelet_parent_r}
      boxExtents=<2 12 12>
      constraintSetup {angularLimitsMax=<0 0 0>, angularLimitsMin=<0 0 0>, angularTarget=<0 0 0>, linearAxesMax=<1.5 0 0>, linearAxesMin=<-1.5 0 0>}
      localJointOffset=<0 0 0>
     }]
    relativeSpaceBone {boneName=spine_05}
   }
   n13 = Anim Dynamics - Bone: dyn_bracelet_upper_parent(ComponentPose=@n12, Alpha=1, ExternalForce="0,0,0") {
    angularDampingOverride=0.9
    bGravityOverrideInSimSpace=true
    bOverrideAngularDamping=true
    bOverrideLinearDamping=true
    linearDampingOverride=0.9
    alphaBoolBlend {blendOption=Linear}
    boundBone {boneName=dyn_bracelet_upper_parent}
    componentAppliedLinearAccClamp=<6000 6000 6000>
    componentLinearAccScale=<0.15 0.15 0.15>
    componentLinearVelScale=<0.15 0.15 0.15>
    physicsBodyDefinitions [{
      sphereCollisionRadius=0
      boundBone {boneName=dyn_bracelet_upper_parent}
      boxExtents=<2 12 12>
      constraintSetup {angularLimitsMax=<0 0 0>, angularLimitsMin=<0 0 0>, angularTarget=<0 0 0>, linearAxesMax=<1.5 0 0>, linearAxesMin=<-1.5 0 0>}
      localJointOffset=<0 0 0>
     }]
    relativeSpaceBone {boneName=spine_05}
   }
   Output Pose(Result=Component To Local(ComponentPose=@n13)) {layerGroup=None}
  ]}
 EventGraph {nodes [
   Event BlueprintInitializeAnimation()
   > Set Skel_Mesh(Skel_Mesh=GetPartSkeletalMeshComponent())
   Event BlueprintUpdateAnimation()
   > n1 = Sequence!(execute) -> (then_0, then_1, then_2:Exec)
   n2 = Get Skel_Mesh()
   > Set ArmRotate_L_Alpha(ArmRotate_L_Alpha:double=GetAlphaFromReferenceTransformDelta(
     SkeletalMeshComponent=@n2
     BoneName=upperarm_l
     BaseName=lowerarm_l
     DeltaComponentType=QuaternionTwist
     InputRange=(X=-5,Y=42)
     TwistAxis="0,0,1"
    )) -> (Output_Get:double)
   Set ArmRotate_R_Alpha(
    execute=@n1.then_1
    ArmRotate_R_Alpha:double=GetAlphaFromReferenceTransformDelta(
     SkeletalMeshComponent=@n2
     BoneName=upperarm_r
     BaseName=lowerarm_r
     DeltaComponentType=QuaternionTwist
     InputRange=(X=-2,Y=84)
     TwistAxis="0,0,1"
    )
   ) -> (Output_Get:double)
   Set Shoulder_R_Alpha(
    execute=@n1.then_2
    Shoulder_R_Alpha:double=GetAlphaFromReferenceTransformDelta(
     SkeletalMeshComponent=@n2
     BoneName=clavicle_r
     BaseName=upperarm_r
     DeltaComponentType=QuaternionTwist
     InputRange=(X=-5,Y=120)
     TwistAxis="0,0,1"
    )
   ) -> (Output_Get:double)
  ]}
]
variables [
 ArmRotate_L_Alpha {type=number}
 ArmRotate_R_Alpha {type=number}
 l_arm_fwd {type=number}
 Shoulder_R_Alpha {type=number}
 Skel_Mesh {type=SkeletalMeshComponent}
]
