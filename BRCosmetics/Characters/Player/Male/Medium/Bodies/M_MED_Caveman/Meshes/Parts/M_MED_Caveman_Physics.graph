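# Physics asset for the male medium Caveman cosmetic body, defining collision and joint limits for thirteen dynamic strands (a–m) and the head.
# Each strand constrains a capsule parent to a small sphere base; strands a, e and g chain further capsules beyond that.
# Constraints use narrow twist and modest swing limits, with strand a's swing loosening toward its tip.
# The large head capsule and the base spheres have no constraints linking them to each other.
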PhysicsAsset /BRCosmetics/Characters/Player/Male/Medium/Bodies/M_MED_Caveman/Meshes/Parts/M_MED_Caveman_Physics
bodies {
 dyn_strand_a_01 [{shapeType=Capsule, length=2.38, radius=1.005, center=<1.8394 0 0>, rotation=<rot 90 0 0>}]
 dyn_strand_a_02 [{shapeType=Capsule, length=2.7233, radius=0.7454, center=<2.0095 0 0>, rotation=<rot 90 0 0>}]
 dyn_strand_a_03 [{shapeType=Capsule, length=4.3185, radius=1.3388, center=<3.1236 -0.0001 0>, rotation=<rot 90 0 0>}]
 dyn_strand_a_base [{shapeType=Sphere, radius=0.5}]
 dyn_strand_a_parent [{shapeType=Capsule, length=1.25, radius=0.88, center=<0.8805 0 0>, rotation=<rot 90 0 0>}]
 dyn_strand_b_base [{shapeType=Sphere, radius=0.5}]
 dyn_strand_b_parent [{shapeType=Capsule, length=1.2297, radius=0.9264, center=<1.0955 0 0>, rotation=<rot 90 0 0>}]
 dyn_strand_c_base [{shapeType=Sphere, radius=0.5}]
 dyn_strand_c_parent [{shapeType=Capsule, length=1.4747, radius=0.8101, center=<0.9327 0 0>, rotation=<rot 90 0 0>}]
 dyn_strand_d_base [{shapeType=Sphere, radius=0.5}]
 dyn_strand_d_parent [{shapeType=Capsule, length=2.3212, radius=0.7464, center=<1.6332 0 0>, rotation=<rot 90 0 0>}]
 dyn_strand_e_01 [{shapeType=Capsule, length=1.255, radius=0.63, center=<0.9538 0 0>, rotation=<rot 90 0 0>}]
 dyn_strand_e_base [{shapeType=Sphere, radius=0.5}]
 dyn_strand_e_parent [{shapeType=Capsule, length=1.5092, radius=0.625, center=<0.8976 0 0>, rotation=<rot 90 0 0>}]
 dyn_strand_f_base [{shapeType=Sphere, radius=0.5}]
 dyn_strand_f_parent [{shapeType=Capsule, length=2.7151, radius=0.9684, center=<1.7365 0 0>, rotation=<rot 90 0 0>}]
 dyn_strand_g_01 [{shapeType=Capsule, length=2.0925, radius=0.7101, center=<1.365 0.0001 0>, rotation=<rot 90 0 0>}]
 dyn_strand_g_base [{shapeType=Sphere, radius=0.5}]
 dyn_strand_g_parent [{shapeType=Capsule, length=2.2093, radius=1.0674, center=<1.4756 0 0>, rotation=<rot 90 0 0>}]
 dyn_strand_h_base [{shapeType=Sphere, radius=0.5}]
 dyn_strand_h_parent [{shapeType=Capsule, length=1.755, radius=1.255, center=<1.41 0 0>, rotation=<rot 90 0 0>}]
 dyn_strand_i_base [{shapeType=Sphere, radius=0.5}]
 dyn_strand_i_parent [{shapeType=Capsule, length=2.9522, radius=0.75, center=<1.9892 0 0>, rotation=<rot 90 0 0>}]
 dyn_strand_j_base [{shapeType=Sphere, radius=0.5}]
 dyn_strand_j_parent [{shapeType=Capsule, length=2.3796, radius=0.8279, center=<1.4243 0 0>, rotation=<rot 90 0 0>}]
 dyn_strand_k_base [{shapeType=Sphere, radius=0.5}]
 dyn_strand_k_parent [{shapeType=Capsule, length=2.755, radius=0.88, center=<1.5649 0 0>, rotation=<rot 90 0 0>}]
 dyn_strand_l_base [{shapeType=Sphere, radius=0.5}]
 dyn_strand_l_parent [{shapeType=Capsule, length=3.0257, radius=0.8994, center=<1.3624 0 -0.0001>, rotation=<rot 90 0 0>}]
 dyn_strand_m_base [{shapeType=Sphere, radius=0.5}]
 dyn_strand_m_parent [{shapeType=Capsule, length=1.7762, radius=0.8276, center=<1.611 0 0>, rotation=<rot 90 0 0>}]
 head [{shapeType=Capsule, length=8.75, radius=6.5, center=<2.4455 3.4801 -0.1322>, rotation=<rot 90 0 0>}]
}
constraints [
 {bone1Name=dyn_strand_a_parent, bone2Name=dyn_strand_a_base, swing1LimitDegrees=5, swing2LimitDegrees=5, twistLimitDegrees=3}
 {bone1Name=dyn_strand_a_01, bone2Name=dyn_strand_a_parent, swing1LimitDegrees=15, swing2LimitDegrees=15, twistLimitDegrees=8}
 {bone1Name=dyn_strand_a_02, bone2Name=dyn_strand_a_01, swing1LimitDegrees=25, swing2LimitDegrees=25, twistLimitDegrees=5}
 {bone1Name=dyn_strand_a_03, bone2Name=dyn_strand_a_02, swing1LimitDegrees=25, swing2LimitDegrees=25, twistLimitDegrees=5}
 {bone1Name=dyn_strand_m_parent, bone2Name=dyn_strand_m_base, swing1LimitDegrees=10, swing2LimitDegrees=20, twistLimitDegrees=5}
 {bone1Name=dyn_strand_l_parent, bone2Name=dyn_strand_l_base, swing1LimitDegrees=20, swing2LimitDegrees=30, twistLimitDegrees=3}
 {bone1Name=dyn_strand_k_parent, bone2Name=dyn_strand_k_base, swing1LimitDegrees=20, swing2LimitDegrees=30, twistLimitDegrees=3}
 {bone1Name=dyn_strand_j_parent, bone2Name=dyn_strand_j_base, swing1LimitDegrees=20, swing2LimitDegrees=20, twistLimitDegrees=3}
 {bone1Name=dyn_strand_i_parent, bone2Name=dyn_strand_i_base, swing1LimitDegrees=20, swing2LimitDegrees=20, twistLimitDegrees=3}
 {bone1Name=dyn_strand_h_parent, bone2Name=dyn_strand_h_base, swing1LimitDegrees=20, swing2LimitDegrees=20, twistLimitDegrees=3}
 {bone1Name=dyn_strand_g_01, bone2Name=dyn_strand_g_parent, swing1LimitDegrees=10, swing2LimitDegrees=10, twistLimitDegrees=3}
 {bone1Name=dyn_strand_g_parent, bone2Name=dyn_strand_g_base, swing1LimitDegrees=15, swing2LimitDegrees=15, twistLimitDegrees=3}
 {bone1Name=dyn_strand_f_parent, bone2Name=dyn_strand_f_base, swing1LimitDegrees=20, swing2LimitDegrees=20, twistLimitDegrees=3}
 {bone1Name=dyn_strand_e_01, bone2Name=dyn_strand_e_parent, swing1LimitDegrees=20, swing2LimitDegrees=20, twistLimitDegrees=3}
 {bone1Name=dyn_strand_e_parent, bone2Name=dyn_strand_e_base, swing1LimitDegrees=10, swing2LimitDegrees=10, twistLimitDegrees=3}
 {bone1Name=dyn_strand_d_parent, bone2Name=dyn_strand_d_base, swing1LimitDegrees=20, swing2LimitDegrees=20, twistLimitDegrees=3}
 {bone1Name=dyn_strand_c_parent, bone2Name=dyn_strand_c_base, swing1LimitDegrees=10, swing2LimitDegrees=10, twistLimitDegrees=3}
 {bone1Name=dyn_strand_b_parent, bone2Name=dyn_strand_b_base, swing1LimitDegrees=20, swing2LimitDegrees=20, twistLimitDegrees=3}
]
