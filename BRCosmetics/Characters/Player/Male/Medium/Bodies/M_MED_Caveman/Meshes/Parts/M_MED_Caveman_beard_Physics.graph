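# Physics asset for the beard mesh part of the medium male Caveman player cosmetic, defining bodies for a beard and a braid.
# Two capsule chains, each starting at a small base sphere, are linked by narrow swing and twist limits, with the braid tightest at its root and loosest at its tip.
# Head, neck, clavicle, upper spine and right lip corner bodies have no constraints and supply body shapes around the chains.
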PhysicsAsset /BRCosmetics/Characters/Player/Male/Medium/Bodies/M_MED_Caveman/Meshes/Parts/M_MED_Caveman_beard_Physics
bodies {
 R_lip_corner [{shapeType=Capsule, length=3.875, radius=2.375, center=<0 -2.1719 -4.715>, rotation=<rot 0 0 -14.4943>}]
 clavicle_l [{shapeType=Capsule, length=12.625, radius=9.125, center=<10.8573 2.6024 -9.7351>, rotation=<rot 5.4787 0 -27.9157>}]
 clavicle_r [{shapeType=Capsule, length=12.625, radius=9.125, center=<-10.1406 -1.911 8.0387>, rotation=<rot 5.4787 0 -27.9157>}]
 dyn_beard_01 [{shapeType=Capsule, length=2.3721, radius=2.9843, center=<2.1438 0 1.4684>, rotation=<rot 0 0 90>}]
 dyn_beard_02 [{shapeType=Capsule, length=1.7193, radius=2.246, center=<1.6727 0 0.6655>, rotation=<rot 0 0 90>}]
 dyn_beard_03 [{shapeType=Capsule, length=2.1501, radius=1.0106, center=<1.0986 0 0.0506>, rotation=<rot 0 0 90>}]
 dyn_beard_base [{shapeType=Sphere, radius=0.5}]
 dyn_beard_parent [{shapeType=Capsule, length=5.2511, radius=3.8966, center=<2.33 0.0001 2.1676>, rotation=<rot 0 0 90>}]
 dyn_braid_01 [{shapeType=Capsule, length=2.7784, radius=0.8908, center=<1.8579 0 0>, rotation=<rot 90 0 0>}]
 dyn_braid_02 [{shapeType=Capsule, length=2.2649, radius=0.8763, center=<1.5368 0 0>, rotation=<rot 90 0 0>}]
 dyn_braid_03 [{shapeType=Capsule, length=3.0754, radius=1.1705, center=<2.1832 -0.0001 0>, rotation=<rot 90 0 0>}]
 dyn_braid_04 [{shapeType=Capsule, length=3.5233, radius=1.2315, center=<1.728 -0.0001 0>, rotation=<rot 90 0 0>}]
 dyn_braid_base [{shapeType=Sphere, radius=0.5}]
 dyn_braid_parent [{shapeType=Capsule, length=2.2506, radius=0.8579, center=<1.7574 0 0>, rotation=<rot 90 0 0>}]
 head [{shapeType=Capsule, length=7.5, radius=6.625, center=<1.2962 1.0579 0>, rotation=<rot 90 0 0>}]
 neck_01 [{shapeType=Capsule, length=5.75, radius=4.75, center=<2.2739 2.624 0>, rotation=<rot 90 0 0>}]
 spine_05 [
  {shapeType=Capsule, length=29.125, radius=5.875, center=<4.6596 0.3008 -0.0001>}
  {shapeType=Box, extentX=26, extentY=8.5, extentZ=41.75, center=<-5.8052 10.3107 0>, rotation=<rot 0 -33.6366 0>}
 ]
}
constraints [
 {bone1Name=dyn_beard_parent, bone2Name=dyn_beard_base, swing1LimitDegrees=10, swing2LimitDegrees=30, twistLimitDegrees=5}
 {bone1Name=dyn_beard_01, bone2Name=dyn_beard_parent, swing1LimitDegrees=10, swing2LimitDegrees=25, twistLimitDegrees=5}
 {bone1Name=dyn_beard_02, bone2Name=dyn_beard_01, swing1LimitDegrees=10, swing2LimitDegrees=30, twistLimitDegrees=5}
 {bone1Name=dyn_beard_03, bone2Name=dyn_beard_02, swing1LimitDegrees=10, swing2LimitDegrees=30, twistLimitDegrees=5}
 {bone1Name=dyn_braid_parent, bone2Name=dyn_braid_base, swing1LimitDegrees=5, swing2LimitDegrees=5, twistLimitDegrees=3}
 {bone1Name=dyn_braid_01, bone2Name=dyn_braid_parent, swing1LimitDegrees=10, swing2LimitDegrees=10, twistLimitDegrees=3}
 {bone1Name=dyn_braid_02, bone2Name=dyn_braid_01, swing1LimitDegrees=10, swing2LimitDegrees=10, twistLimitDegrees=3}
 {bone1Name=dyn_braid_03, bone2Name=dyn_braid_02, swing1LimitDegrees=10, swing2LimitDegrees=10, twistLimitDegrees=3}
 {bone1Name=dyn_braid_04, bone2Name=dyn_braid_03, swing1LimitDegrees=20, swing2LimitDegrees=20, twistLimitDegrees=3}
]
