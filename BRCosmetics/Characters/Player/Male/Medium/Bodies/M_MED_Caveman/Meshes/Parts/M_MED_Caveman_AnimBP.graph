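# Animation Blueprint for the Caveman cosmetic body part, adding secondary motion and head-driven beard correction to the pose copied from the attached parent mesh.
# Two chained RigidBody nodes simulate the beard and general physics assets relative to spine_05.
# Head twist against spine_05 sets HeadRot_Alpha, which blends an additive offset on the beard base and a Look At aiming dyn_beard_03 at spine_03.
# RotSides feeds only a disconnected bone-modify node.
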AnimBlueprint /BRCosmetics/Characters/Player/Male/Medium/Bodies/M_MED_Caveman/Meshes/Parts/M_MED_Caveman_AnimBP
target_skeleton=&/M_MED_Caveman_Skeleton
graphs [
 AnimGraph {nodes [
   "Transform (Modify) Bone - Bone: dyn_beard_base"(Alpha=Get RotSides() -> (RotSides:double), Translation="0,0,0", Rotation="-10,0,0", Scale="1,1,1") {
    rotationMode=BMM_Additive
    rotationSpace=BCS_BoneSpace
    translationMode=BMM_Additive
    translationSpace=BCS_BoneSpace
    alphaBoolBlend {blendOption=Linear}
    alphaScaleBiasClamp {bInterpResult=true, interpSpeedDecreasing=2, interpSpeedIncreasing=3}
    boneToModify {boneName=dyn_beard_base}
    rotation {pitch=-10}
   }
   n1 = RigidBody(ComponentPose=Local To Component(LocalPose=Copy Pose From Mesh() {bUseAttachedParent=true})) {
    bClampLinearTranslationLimitToRefPose=true
    bOverrideWorldGravity=true
    overridePhysicsAsset=&/M_MED_Caveman_beard_Physics
    simulationSpace=BaseBoneSpace
    alphaBoolBlend {blendOption=Linear}
    alphaScaleBiasClamp {interpSpeedDecreasing=4, interpSpeedIncreasing=6}
    baseBoneRef {boneName=spine_05}
    componentAppliedLinearAccClamp=<3000 3000 3000>
    componentLinearAccScale=<0.2 0.2 0.2>
    componentLinearVelScale=<0.2 0.2 0.2>
    overrideWorldGravity {z=-980}
   }
   n2 = RigidBody(ComponentPose=@n1) {
    bClampLinearTranslationLimitToRefPose=true
    bOverrideWorldGravity=true
    overridePhysicsAsset=&/M_MED_Caveman_Physics
    simulationSpace=BaseBoneSpace
    alphaBoolBlend {blendOption=Linear}
    baseBoneRef {boneName=spine_05}
    componentAppliedLinearAccClamp=<3000 3000 3000>
    componentLinearAccScale=<0.2 0.2 0.2>
    componentLinearVelScale=<0.2 0.2 0.2>
    overrideWorldGravity {z=-980}
   }
   n3 = Get HeadRot_Alpha() -> (HeadRot_Alpha:double)
   n4 = "Transform (Modify) Bone - Bone: dyn_beard_base"(ComponentPose=@n2, Alpha=@n3, Translation="-1,0,-1", Rotation="-5,0,0", Scale="1,1,1") {
    rotationMode=BMM_Additive
    rotationSpace=BCS_BoneSpace
    translationMode=BMM_Additive
    translationSpace=BCS_BoneSpace
    alphaBoolBlend {blendOption=Linear}
    alphaScaleBiasClamp {bInterpResult=true, interpSpeedDecreasing=4, interpSpeedIncreasing=6}
    boneToModify {boneName=dyn_beard_base}
    rotation {pitch=-5}
    translation {x=-1, z=-1}
   }
   n5 = Look At - Bone: dyn_beard_03(ComponentPose=@n4, Alpha=@n3) {
    lookAtClamp=1
    alphaBoolBlend {blendOption=Linear}
    alphaScaleBiasClamp {interpSpeedDecreasing=4, interpSpeedIncreasing=6}
    boneToModify {boneName=dyn_beard_03}
    lookAtLocation {x=-50}
    lookAtTarget {boneReference {boneName=spine_03}}
    lookAt_Axis {axis=<1 0>}
    lookUp_Axis {axis {x=-1}}
   }
   Output Pose(Result=Component To Local(ComponentPose=@n5)) {layerGroup=None}
  ]}
 EventGraph {nodes [
   Event BlueprintUpdateAnimation()
   > n1 = Sequence()
   > Set HeadRot_Alpha(HeadRot_Alpha:double=GetAlphaFromReferenceTransformDelta(
     SkeletalMeshComponent=Get Skel_Mesh()
     BoneName=head
     BaseName=spine_05
     DeltaComponentType=QuaternionTwist
     InputRange=(X=34,Y=0)
     TwistAxis="0,0,1"
    )) -> (Output_Get:double)
   n2 = GetDeltaTransformFromRefPose(self=GetOwnerSkeletalMeshComponent(), BoneName=head, BaseName=spine_05)
   Set RotSides(execute=@n1.then_1, RotSides:double=BreakRotator(InRot=BreakTransform(InTransform=@n2).Rotation).Pitch) -> (Output_Get:double)
   Event BlueprintInitializeAnimation()
   > Set Skel_Mesh(Skel_Mesh=GetPartSkeletalMeshComponent())
   GetPartSkeletalMeshComponent()
   GetOwnerSkeletalMeshComponent()
  ]}
]
variables [
 HeadRot_Alpha {type=number}
 RotSides {type=number}
 Skel_Mesh {type=SkeletalMeshComponent}
]
